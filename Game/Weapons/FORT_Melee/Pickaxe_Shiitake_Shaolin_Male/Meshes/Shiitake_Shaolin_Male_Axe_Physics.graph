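# Physics asset for the Shiitake Shaolin Male pickaxe mesh, setting up collision and joint limits for its two dangling ribbons.
# A capsule covers the two-handed handle, a box sits on the ribbon master bone, and small boxes cover each ribbon segment.
# Each ribbon is a two-link chain constrained to the master bone, with asymmetric swing and tighter twist at the root and uniform limits at the tip.
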PhysicsAsset /Game/Weapons/FORT_Melee/Pickaxe_Shiitake_Shaolin_Male/Meshes/Shiitake_Shaolin_Male_Axe_Physics
bodies {
 Melee_TwoHanded_Handle [{shapeType=Capsule, length=43.3699, radius=20.0232, center=<0 0 77.5895>}]
 dyn_ribbon_a_01 [{shapeType=Box, extentX=1.1, extentY=4, extentZ=5, center=<2.5 0 0>, rotation=<rot 90 0 -0.184>}]
 dyn_ribbon_a_02 [{shapeType=Box, extentX=1.1, extentY=4, extentZ=5, center=<2.5 0 0>, rotation=<rot 90 0 0>}]
 dyn_ribbon_b_01 [{shapeType=Box, extentX=1.1, extentY=4, extentZ=5, center=<2.5 0 0>, rotation=<rot 90 0 -0.184>}]
 dyn_ribbon_b_02 [{shapeType=Box, extentX=1.1, extentY=4, extentZ=5, center=<2.5 0 0>, rotation=<rot 90 0 0>}]
 dyn_ribbon_mstr [{shapeType=Box, extentX=5.8765, extentY=5.8765, extentZ=5.8765}]
}
constraints [
 {bone1Name=dyn_ribbon_a_01, bone2Name=dyn_ribbon_mstr, swing1LimitDegrees=25, swing2LimitDegrees=45, twistLimitDegrees=15}
 {bone1Name=dyn_ribbon_a_02, bone2Name=dyn_ribbon_a_01, swing1LimitDegrees=30, swing2LimitDegrees=30, twistLimitDegrees=30}
 {bone1Name=dyn_ribbon_b_01, bone2Name=dyn_ribbon_mstr, swing1LimitDegrees=25, swing2LimitDegrees=45, twistLimitDegrees=15}
 {bone1Name=dyn_ribbon_b_02, bone2Name=dyn_ribbon_b_01, swing1LimitDegrees=30, swing2LimitDegrees=30, twistLimitDegrees=30}
]
